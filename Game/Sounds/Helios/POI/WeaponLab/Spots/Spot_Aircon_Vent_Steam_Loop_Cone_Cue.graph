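# Looping ambient spot Sound Cue for an air-conditioning vent with steam, filed under Helios Weapon Lab POI sounds.
# A Mixer layers a modulated random pick of two AC duct air loops with a random pick of a vault door ambience or pipe fitting loop.
# All waves loop and are reused from Asteria and Artemis content.
# It uses the ambient sound class, its own cone-named attenuation asset and a small concurrency cap.
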SoundCue /Game/Sounds/Helios/POI/WeaponLab/Spots/Spot_Aircon_Vent_Steam_Loop_Cone_Cue
root=Mixer_1
nodes [
 {id=Mixer_1, children [Modulator_0,Random_0], properties {inputVolume [1,1]}}
 {id=Modulator_0, children [Random_1]}
 {id=Random_0, children [WavePlayer_3,WavePlayer_4], properties {weights [1,1]}}
 {id=Random_1, children [WavePlayer_0,WavePlayer_2], properties {weights [1,1]}}
 {id=WavePlayer_0, properties {bLooping=true, soundWaveAssetPtr=/Game/Sounds/Asteria/POI/IndustrialTown/Spots/Spot_AC_Duct_Air_03}}
 {id=WavePlayer_2, properties {bLooping=true, soundWaveAssetPtr=/Game/Sounds/Asteria/POI/IndustrialTown/Spots/Spot_AC_Duct_Air_02}}
 {id=WavePlayer_3, properties {bLooping=true, soundWaveAssetPtr=/Game/Sounds/Artemis/POI/Hollow_Base/Spots/Spot_Vault_Door_Ambient_Loop}}
 {
  id=WavePlayer_4
  properties {bLooping=true, soundWaveAssetPtr=/Game/Sounds/Artemis/Biomes/Grasslands/Grassland_Spots/Spot_Outpost_Pipe_Fitting_Loop}
 }
]
properties {
 attenuationSettings=&/Spot_Aircon_Vent_Steam_Loop_Cone_Att
 bOverrideConcurrency=true
 maxDistance=3500
 soundClassObject=&Fort_Ambient_Athena
 volumeMultiplier=0.301
 attenuationOverrides {attenuationShapeExtents {x=0}}
 concurrencyOverrides {maxCount=3}
}
