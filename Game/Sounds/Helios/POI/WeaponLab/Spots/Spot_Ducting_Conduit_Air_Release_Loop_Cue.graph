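# Looping ambient spot sound cue for an air release from ducting conduit in the Helios Weapon Lab POI.
# A looping stereo wave feeds a single-input Mixer and a default Modulator, playing at reduced volume in the indoor ambient sound class.
# Overridden attenuation uses a custom falloff curve, non-spatialized blend radii and interpolated listener focus that lowers off-focus volume.
# Overridden concurrency allows one instance per owner and blocks new ones.
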SoundCue /Game/Sounds/Helios/POI/WeaponLab/Spots/Spot_Ducting_Conduit_Air_Release_Loop_Cue
root=Modulator_1
nodes [
 {id=Mixer_0, children [WavePlayer_0], properties {inputVolume [1]}}
 {id=Modulator_1, children [Mixer_0]}
 {
  id=WavePlayer_0
  properties {bLooping=true, soundWaveAssetPtr=&/Spot_Ducting_Conduit_Air_Release_Loop_Stereo}
 }
]
properties {
 bOverrideAttenuation=true
 bOverrideConcurrency=true
 maxDistance=1875
 soundClassObject=&Fort_Ambient_Indoors_Athena
 volumeMultiplier=0.501
 attenuationOverrides {
  bEnableFocusInterpolation=true
  bEnableListenerFocus=true
  distanceAlgorithm=Custom
  falloffDistance=1500
  focusAzimuth=25
  nonFocusAzimuth=55
  nonFocusVolumeAttenuation=0.4
  nonSpatializedRadiusEnd=1000
  nonSpatializedRadiusStart=4000
  attenuationShapeExtents=<0 150>
  customAttenuationCurve {editorCurveData {keys [<at 0 1>,<at 0.2 1>,<at 0.5014 0.4978>,<at 0.5547 0.3739>,<at 0.7502 0.2097>,<at 1.25 0>]}}
 }
 busSends [{audioBus=null, minSendLevel=0, sendLevel=1, soundSourceBus=null}]
 concurrencyOverrides {bLimitToOwner=true, maxCount=1, resolutionRule=PreventNew}
}
