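# Looping indoor ambient spot sound cue of flapping ducting conduit for the Weapon Lab POI in the Helios sounds.
# A looping stereo wave player feeds a default-setting Modulator, with output at about half volume in the indoor ambient sound class.
# Overridden attenuation uses a capsule shape with zero stereo spread, distance-based filtering and listener focus that quiets unfocused playback.
# Overridden concurrency prevents new instances once its limit is reached.
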SoundCue /Game/Sounds/Helios/POI/WeaponLab/Spots/Spot_Ducting_Conduit_Flapping_Loop_Cue
root=Modulator_1
nodes [
 {id=Modulator_1, children [WavePlayer_1]}
 {
  id=WavePlayer_1
  properties {bLooping=true, soundWaveAssetPtr=&/Spot_Ducting_Conduit_Flapping_Loop_Stereo}
 }
]
properties {
 bOverrideAttenuation=true
 bOverrideConcurrency=true
 maxDistance=1550
 soundClassObject=&Fort_Ambient_Indoors_Athena
 volumeMultiplier=0.501
 attenuationOverrides {
  attenuationShape=Capsule
  bAttenuateWithLPF=true
  bEnableFocusInterpolation=true
  bEnableListenerFocus=true
  falloffDistance=1250
  hPFFrequencyAtMax=750
  hPFFrequencyAtMin=250
  lPFRadiusMax=1500
  lPFRadiusMin=500
  nonFocusVolumeAttenuation=0.25
  stereoSpread=0
  attenuationShapeExtents {x=300}
 }
 concurrencyOverrides {resolutionRule=PreventNew}
}
