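# Indoor ambient spot cue for tonal whooshes in Weapon Lab ducting conduits.
# An enveloped, pitched-up loop plays under a quieter looping layer that randomly picks two or three delayed groups of random one-shot whooshes.
# The first two quality levels share this mix; the other two are empty.
# Overrides set single-instance concurrency, cone attenuation with listener focus and filtering, plus reflection bus and reverb sends.
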SoundCue /Game/Sounds/Helios/POI/WeaponLab/Spots/Spot_Ducting_Conduit_Tonal_Whoosh_Cue
root=QualityLevel_0
nodes [
 {id=Delay_0, children [Modulator_0], properties {delayMax=0.85, delayMin=0.66}}
 {id=Delay_1, children [Modulator_2], properties {delayMax=0.66, delayMin=0.25}}
 {id=Delay_2, children [Modulator_3], properties {delayMax=0.33}}
 {
  id=Enveloper_0
  children [Modulator_1]
  properties {
   loopEnd=1.5
   pitchCurve {editorCurveData {keys [{time=0}]}}
   volumeCurve {editorCurveData {keys [<at 0 1 RCIM_Cubic>,<at 0.75 0.6 RCIM_Cubic>,<at 1.5 1>]}}
  }
 }
 {id=Looping_3, children [Random_6]}
 {id=Mixer_0, children [Looping_3,Enveloper_0], properties {inputVolume [0.3,1]}}
 {id=Mixer_1, children [Random_0], properties {inputVolume [1]}}
 {id=Mixer_2, children [Random_1], properties {inputVolume [1]}}
 {id=Mixer_3, children [Random_2], properties {inputVolume [1.45]}}
 {id=Mixer_4, children [Delay_1,Delay_2], properties {inputVolume [1,1]}}
 {id=Mixer_5, children [Delay_2,Delay_1,Delay_0], properties {inputVolume [1,1,1]}}
 {id=Modulator_0, children [Mixer_1], properties {pitchMax=1, pitchMin=1}}
 {id=Modulator_1, children [WavePlayer_0], properties {pitchMax=1.1, pitchMin=1.1}}
 {id=Modulator_2, children [Mixer_2], properties {pitchMax=1, pitchMin=1}}
 {id=Modulator_3, children [Mixer_3], properties {pitchMin=1}}
 {id=QualityLevel_0, children [Mixer_0,Mixer_0,null,null]}
 {id=Random_0, children [WavePlayer_4, WavePlayer_5, WavePlayer_11, WavePlayer_12, WavePlayer_13], properties {weights [1,1,1,1,1]}}
 {id=Random_1, children [WavePlayer_8, WavePlayer_9, WavePlayer_10, WavePlayer_7], properties {weights [1,1,1,1]}}
 {id=Random_2, children [WavePlayer_2,WavePlayer_3], properties {weights [1,1]}}
 {id=Random_6, children [Mixer_4,Mixer_5], properties {weights [1,1]}}
 {
  id=WavePlayer_0
  properties {bLooping=true, soundWaveAssetPtr=&/Spot_Ducting_Conduit_Tonal_Whoosh_Loop}
 }
 {id=WavePlayer_10, properties {soundWaveAssetPtr=&/Spot_Ducting_Conduit_Tonal_Whoosh_04}}
 {id=WavePlayer_11, properties {soundWaveAssetPtr=&/Spot_Ducting_Conduit_Tonal_Whoosh_03}}
 {id=WavePlayer_12, properties {soundWaveAssetPtr=&/Spot_Ducting_Conduit_Tonal_Whoosh_02}}
 {id=WavePlayer_13, properties {soundWaveAssetPtr=&/Spot_Ducting_Conduit_Tonal_Whoosh_01}}
 {id=WavePlayer_2, properties {soundWaveAssetPtr=&/Spot_Ducting_Conduit_Tonal_Whoosh_12}}
 {id=WavePlayer_3, properties {soundWaveAssetPtr=&/Spot_Ducting_Conduit_Tonal_Whoosh_11}}
 {id=WavePlayer_4, properties {soundWaveAssetPtr=&/Spot_Ducting_Conduit_Tonal_Whoosh_10}}
 {id=WavePlayer_5, properties {soundWaveAssetPtr=&/Spot_Ducting_Conduit_Tonal_Whoosh_09}}
 {id=WavePlayer_7, properties {soundWaveAssetPtr=&/Spot_Ducting_Conduit_Tonal_Whoosh_07}}
 {id=WavePlayer_8, properties {soundWaveAssetPtr=&/Spot_Ducting_Conduit_Tonal_Whoosh_06}}
 {id=WavePlayer_9, properties {soundWaveAssetPtr=&/Spot_Ducting_Conduit_Tonal_Whoosh_05}}
]
properties {
 bOverrideAttenuation=true
 bOverrideConcurrency=true
 maxDistance=1100
 soundClassObject=&Fort_Ambient_Indoors_Athena
 volumeMultiplier=0.301
 attenuationOverrides {
  attenuationShape=Cone
  bAttenuateWithLPF=true
  bEnableFocusInterpolation=true
  bEnableListenerFocus=true
  distanceAlgorithm=NaturalSound
  falloffDistance=750
  focusAzimuth=15
  focusReleaseInterpSpeed=0.8
  hPFFrequencyAtMax=500
  hPFFrequencyAtMin=100
  lPFRadiusMax=1000
  lPFRadiusMin=350
  nonFocusAzimuth=45
  nonFocusVolumeAttenuation=0.4
  nonSpatializedRadiusEnd=150
  nonSpatializedRadiusStart=250
  stereoSpread=10
  attenuationShapeExtents=<350 90 120>
  customAttenuationCurve {editorCurveData {keys [<at 0 1>,<at 0.2 1>,<at 0.5014 0.4978>,<at 0.5547 0.3739>,<at 0.7502 0.2097>,<at 1.25 0>]}}
 }
 busSends [{
   audioBus=null
   maxSendDistance=1500
   minSendLevel=0
   sendLevel=1
   soundSourceBus=/Game/Sounds/Apollo/Biomes/Reflections/AmbientAudio_Reflection_SourceBus
   customSendLevelCurve {editorCurveData {keys [<at 0 0.2>,<at 1 1>]}}
  }]
 concurrencyOverrides {maxCount=1, resolutionRule=PreventNew, voiceStealReleaseTime=0.33}
 soundSubmixSends [{
   maxSendDistance=1500
   minSendLevel=0.4
   soundSubmix=&Fortnite_Reverb_Default
   customSendLevelCurve {editorCurveData {keys [<at 0 0.2>,<at 1 1>]}}
  }]
}
