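# Looping ambient spot cue for a tonal whoosh in ducting conduit at the Helios WeaponLab POI, classed as indoor ambience.
# A single looping wave player feeds the output, pitched up and reduced in volume.
# Overridden attenuation adds a custom falloff curve and listener focus that quiets off-axis sound.
# Reflection bus and reverb submix sends rise with distance, while concurrency allows two instances and blocks new ones.
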SoundCue /Game/Sounds/Helios/POI/WeaponLab/Spots/Spot_Ducting_Conduit_Tonal_Whoosh_Loop_Cue
root=WavePlayer_0
nodes [{
  id=WavePlayer_0
  properties {bLooping=true, soundWaveAssetPtr=&/Spot_Ducting_Conduit_Tonal_Whoosh_Loop}
 }]
properties {
 bOverrideAttenuation=true
 bOverrideConcurrency=true
 maxDistance=1500
 pitchMultiplier=2
 soundClassObject=&Fort_Ambient_Indoors_Athena
 volumeMultiplier=0.501
 attenuationOverrides {
  bEnableFocusInterpolation=true
  bEnableListenerFocus=true
  falloffDistance=1500
  focusAzimuth=25
  nonFocusAzimuth=55
  nonFocusVolumeAttenuation=0.4
  nonSpatializedRadiusEnd=150
  nonSpatializedRadiusStart=250
  stereoSpread=10
  attenuationShapeExtents=<0 150>
  customAttenuationCurve {editorCurveData {keys [<at 0 1>,<at 0.2 1>,<at 0.5014 0.4978>,<at 0.5547 0.3739>,<at 0.7502 0.2097>,<at 1.25 0>]}}
 }
 busSends [{
   audioBus=null
   maxSendDistance=1500
   minSendLevel=0
   sendLevel=1
   soundSourceBus=/Game/Sounds/Apollo/Biomes/Reflections/AmbientAudio_Reflection_SourceBus
   customSendLevelCurve {editorCurveData {keys [<at 0 0.2>,<at 1 1>]}}
  }]
 concurrencyOverrides {maxCount=2, resolutionRule=PreventNew, voiceStealReleaseTime=0.33}
 soundSubmixSends [{
   maxSendDistance=1500
   minSendLevel=0.2
   soundSubmix=&Fortnite_Reverb_Default
   customSendLevelCurve {editorCurveData {keys [<at 0 0.2>,<at 1 1>]}}
  }]
}
