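# Looping spot ambience cue that plays randomized door bangs for the Weapon Lab POI at irregular, delayed intervals.
# Each pass picks a plain bang, a volume-boosted one, or a less frequent layered hit.
# The layered hit mixes an envelope-faded bang from a second sample pool with a slightly delayed bang from the main pool.
# Output uses custom attenuation, limited concurrency, an exterior ambient bus send and a reverb send.
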SoundCue /Game/Sounds/Helios/POI/WeaponLab/Spots/Spot_Weapon_Lab_Door_Bangs_Cue
root=Looping_2
nodes [
 {id=Delay_2, children [Modulator_3], properties {delayMax=5, delayMin=1}}
 {id=Delay_3, children [Random_2], properties {delayMax=0.5, delayMin=0.3}}
 {
  id=Enveloper_0
  children [Random_3]
  properties {pitchCurve {editorCurveData {keys [{time=0}]}}, volumeCurve {editorCurveData {keys [<at 0 1>,<at 1 1>,<at 3 0>]}}}
 }
 {id=Looping_2, children [Delay_2]}
 {id=Mixer_1, children [Delay_3,Enveloper_0], properties {inputVolume [1,1]}}
 {id=Modulator_2, children [Random_2], properties {pitchMax=1, pitchMin=1, volumeMax=1.2, volumeMin=1}}
 {id=Modulator_3, children [Random_4], properties {volumeMax=1, volumeMin=1}}
 {
  id=Random_2
  children [WavePlayer_0, WavePlayer_13, WavePlayer_15, WavePlayer_16, WavePlayer_20, WavePlayer_21, WavePlayer_23, WavePlayer_24]
  properties {bShouldExcludeFromBranchCulling=true, weights [1,1,1,1,1,1,1,1]}
 }
 {id=Random_3, children [WavePlayer_17, WavePlayer_22, WavePlayer_19, WavePlayer_18, WavePlayer_14], properties {weights [1,1,1,1,1]}}
 {id=Random_4, children [Modulator_2,Random_2,Mixer_1], properties {bShouldExcludeFromBranchCulling=true, weights [1,1,0.25]}}
 {id=WavePlayer_0, properties {soundWaveAssetPtr=&/Spot_Weapon_Lab_Door_Bang_01}}
 {id=WavePlayer_13, properties {soundWaveAssetPtr=&/Spot_Weapon_Lab_Door_Bang_02}}
 {id=WavePlayer_14, properties {soundWaveAssetPtr=&/Spot_Weapon_Lab_Door_Bang_03}}
 {id=WavePlayer_15, properties {soundWaveAssetPtr=&/Spot_Weapon_Lab_Door_Bang_04}}
 {id=WavePlayer_16, properties {soundWaveAssetPtr=&/Spot_Weapon_Lab_Door_Bang_05}}
 {id=WavePlayer_17, properties {soundWaveAssetPtr=&/Spot_Weapon_Lab_Door_Bang_06}}
 {id=WavePlayer_18, properties {soundWaveAssetPtr=&/Spot_Weapon_Lab_Door_Bang_07}}
 {id=WavePlayer_19, properties {soundWaveAssetPtr=&/Spot_Weapon_Lab_Door_Bang_08}}
 {id=WavePlayer_20, properties {soundWaveAssetPtr=&/Spot_Weapon_Lab_Door_Bang_09}}
 {id=WavePlayer_21, properties {soundWaveAssetPtr=&/Spot_Weapon_Lab_Door_Bang_10}}
 {id=WavePlayer_22, properties {soundWaveAssetPtr=&/Spot_Weapon_Lab_Door_Bang_11}}
 {id=WavePlayer_23, properties {soundWaveAssetPtr=&/Spot_Weapon_Lab_Door_Bang_12}}
 {id=WavePlayer_24, properties {soundWaveAssetPtr=&/Spot_Weapon_Lab_Door_Bang_13}}
]
properties {
 attenuationSettings=&/Spot_Weapon_Lab_Door_Bangs_Att
 bOverrideConcurrency=true
 maxDistance=6500
 soundClassObject=&Fort_Doors
 volumeMultiplier=0.4
 busSends [{
   audioBus=null
   minSendLevel=0
   sendLevel=1
   soundSourceBus=&AmbientAudio_Exterior_SourceBus
   customSendLevelCurve {editorCurveData {keys [<at 0 0.4>,<at 1 1>]}}
  }]
 concurrencyOverrides {maxCount=2, volumeScaleMode=Distance}
 preEffectBusSends [{soundSourceBus=null}]
 soundSubmixSends [{
   maxSendDistance=3000
   maxSendLevel=0.6
   minSendDistance=500
   minSendLevel=0.25
   soundSubmix=&Fortnite_Reverb_Default
   customSendLevelCurve {editorCurveData {keys [<at 0 0.3>,<at 1 1>]}}
  }]
}
